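# Niagara system for the purple VinderTech Headset glider, adding glowing eyes, eye trails and magenta-purple flames and ash.
# Eye glows with light attributes follow head sockets, while mesh trails follow the head joint with smoothed positions sent to materials.
# Flame emitters pin sprites to bones by spawn group, using emitter-age curves to stagger bursts along the mesh.
# Front-end modifiers switch or retune effects between menu and gameplay, and time-of-day scalars adjust brightness.
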NiagaraSystem /BRCosmetics/Gadgets/Assets/VinderTech_GliderChute/Glider_Headset/FX/Purple/NS_Glider_Headset_Purple
dependencies {
 usedDataInterfaces [
  &NiagaraDataInterfaceArrayFloat3
  &NiagaraDataInterfaceCurve
  &NiagaraDataInterfaceArrayFloat
  &NiagaraDataInterfaceSpriteRendererInfo
 ]
 usedDynamicInputs [&SelectFloatFromArray]
 usedModules [&SystemState]
}
emitters {
 Ash {
  emitterUpdateScript [
   &EmitterState(
    Delay First Loop Only=true
    Inactive Response=::NewEnumerator1
    Life Cycle Mode=::NewEnumerator1
    Loop Delay=&FN_FrontEndFloatModifier(Gameplay=2.5, Front End=3.5)
    Loop Duration=1
    UseLoopDelay=true
   )
   &SpawnRate(SpawnRate=&RandomRangeFloat(Minimum=70, Maximum=80), Use Spawn Probability=false)
  ]
  particleSpawnScript [
   &FN_SetSpawnGroup(Number of Spawn Groups=22)
   /SetVariables(Particles.New Normalize Spawn Group:NiagaraFloat=&Divide_Float(
     A=&MakeFloatFromInt(Int=&Multiply_Int(A=$Particles.SpawnGroup, B=2))
     B=43
    )) {isSetParameters=true, module=SetVariables_#1}
   &InitializeParticle(
    Color=&RandomRangeLinearColor(Minimum=<rgba 8 0.8 7.4 1>, Maximum=<rgba 6 0.6 5.1 1>)
    Lifetime Max=0.8
    Lifetime Min=0.45
    Lifetime Mode=::NewEnumerator1
    Sprite Rotation Mode=::NewEnumerator2
    Sprite Size=&Multiply_Vector2D_ByFloat(
     Float=&Multiply_Float(
      A=&RandomRangeFloat(Minimum=8, Maximum=6)
      B=&FN_FrontEndFloatModifier(Gameplay=0.3, Front End=0.1)
     )
     Vector2D=&RandomRangeVector2D(Minimum=<2 6>, Maximum=<3 7>)
    )
    Sprite Size Mode=::NewEnumerator4
    UsePositionOffset=false
    Write Material Random=true
   )
   &CameraOffset(Camera Offset Amount=10)
   &SkeletalMeshLocation(
    Bone Sampling Mode=::NewEnumerator0
    MaskSpawnGroup=false
    OverrideRandomSeed=false
    Sampled Position Offset=&MakeVector(
     X=&RandomRangeFloat(Minimum=0, Maximum=0)
     Y=&RandomRangeFloat(Minimum=0, Maximum=0)
     Z=&FloatFromCurve(
      FloatCurve=DataInterface() {propertyValues {Curve {keys [<at 0 0>,<at 0.05 0.4524>,<at 0.2838 0.6524>,<at 0.45 0>]}}}
      CurveIndex=$Particles.New Normalize Spawn Group
      Scale Curve=80
     )
    )
    Skeletal Mesh=$User.Skeletal Mesh
   )
   &AddVelocity(Velocity=&FN_FrontEndVector3Modifier(
     Gameplay=&RandomRangeVector(Minimum=<0 150 350>, Maximum=<0 200 150>)
     Front End=&RandomRangeVector(Minimum=<0 -150 200>, Maximum=<0 -50 100>)
    ))
  ]
  particleUpdateScript [
   &ParticleState()
   &ScaleSpriteSize(
    Scale Sprite Size Mode=::NewEnumerator0
    Uniform Scale Factor=&FloatFromCurve(
     FloatCurve=DataInterface() {propertyValues {Curve {keys [
         {arriveTangent=1.9809, interpMode=RCIM_Cubic, leaveTangent=1.9809, tangentMode=RCTM_User, time=0, value=0}
         {arriveTangent=-5.0021, interpMode=RCIM_Cubic, leaveTangent=-5.0021, tangentMode=RCTM_User, time=0.2, value=1.5}
         <at 1 0>
        ]}}}
     Scale Curve=&FN_FrontEndFloatModifier(Gameplay=0.3, Front End=0.3)
    )
   )
   &ScaleColor(
    Scale Alpha=&FloatFromCurve(FloatCurve=DataInterface() {propertyValues {Curve {keys [
         {interpMode=RCIM_Cubic, tangentMode=RCTM_User, time=0, value=0}
         {interpMode=RCIM_Cubic, tangentMode=RCTM_User, time=0.2, value=1}
         {arriveTangent=-2.2917, interpMode=RCIM_Cubic, leaveTangent=-2.2917, tangentMode=RCTM_User, time=1, value=0}
        ]}}})
    Scale RGB=&VectorFromFloat(Value=&FloatFromCurve(
      FloatCurve=DataInterface() {propertyValues {Curve {keys [{arriveTangent=-0.735, interpMode=RCIM_Cubic, leaveTangent=-0.735, tangentMode=RCTM_User, time=0, value=1}, <at 1 0>]}}}
      Scale Curve=&FN_FrontEndFloatModifier(Gameplay=1, Front End=2)
     ))
    ScaleA=true
    ScaleRGB=true
   )
   &CurlNoiseForce(Noise Frequency=20, Noise Strength=400, Pan Noise=false)
   &Drag(Drag=2, Use Linear Drag=true, Use Rotational Drag=true)
   &SolveForcesAndVelocity(Clamp Velocity=false, Limit Acceleration=false)
   /SetVariables(Particles.Glow03_Color=&MakeLinearColorFromVectorAndFloat(
     "Vector (RGB)"=&MakeVectorFromLinearColorRGB(Color=&MultiplyLinearColor(
       Linear Color=<rgba 0.4398 0.0396 1 0.2>
       Scale Factor=&MakeVector4_FromFloat(Float=&Multiply_Float(
         A=&FNTimeOfDayScalar(Night Scale=0.7, Morning Scale=0.2, Evening Scale=0.2)
         B=&FN_FrontEndFloatModifier(Gameplay=1, Front End=1)
        ))
      ))
     "Float (Alpha)"=&FN_FrontEndFloatModifier(Gameplay=0.2, Front End=0.1)
    )) {isSetParameters=true}
   /SetVariables(Particles.Glow03_Size=&Multiply_Vector2D_ByFloat(
     Float=&FloatFromCurve(
      FloatCurve=DataInterface() {propertyValues {Curve {keys [
          <at 0 0 RCIM_Cubic>
          <at 0.2 1.3 RCIM_Cubic>
          {arriveTangent=-0.1423, interpMode=RCIM_Cubic, leaveTangent=-0.1423, tangentMode=RCTM_User, time=1, value=0}
         ]}}}
      Scale Curve=10
     )
     Vector2D=<1 2>
    )) {isSetParameters=true}
  ]
  rendererClasses [&NiagaraSpriteRendererProperties]
  renderers [
   {rendererClass=&NiagaraSpriteRendererProperties}
   {rendererClass=&NiagaraSpriteRendererProperties, rendererIndex=1}
  ]
 }
 Eyes_Glow_01 {
  emitterUpdateScript [
   &EmitterState(
    Inactive Response=::NewEnumerator1
    Life Cycle Mode=::NewEnumerator1
    Loop Behavior=::NewEnumerator1
    Loop Duration Mode=::NewEnumerator1
    UseLoopDelay=true
   )
   &SpawnBurst_Instantaneous(Spawn Count=2, Use Spawn Probability=false)
  ]
  particleSpawnScript [
   &FN_SetSpawnGroup(Number of Spawn Groups=2)
   &InitializeParticle(
    Color=<rgba 10 0 8.8333 1>
    Sprite Rotation Mode=::NewEnumerator2
    Sprite Size Mode=::NewEnumerator3
    Uniform Sprite Size=40
    UsePositionOffset=false
    Write Material Random=true
   )
   &CameraOffset(Camera Offset Amount=&FN_FrontEndFloatModifier(Gameplay=1, Front End=5))
  ]
  particleUpdateScript [
   &ParticleState(Kill Particles When Lifetime Has Elapsed=false, Let Infinitely Lived Particles Die When Emitter Deactivates=true)
   &SkeletalMeshLocation(
    Bone / Socket Index=$Particles.SpawnGroup
    MaskSpawnGroup=false
    Mesh Sampling Type=::NewEnumerator5
    Skeletal Mesh=DataInterface() {propertyValues {
      DefaultMesh=null
      ExcludeBoneName=root
      OptionalComponentClass=null
      PreviewMesh=/BRCosmetics/Gadgets/Assets/VinderTech_GliderChute/Glider_Headset/Meshes/Headset_Glider
      SkinningMode=SkinOnTheFly
      SoftSourceActor=null
      SourceMode=Default
      UvSetIndex=0
      WholeMeshLOD=-1
      bExcludeBone=false
      bReadDeformedGeometry=true
      bRequireCurrentFrameData=true
      ComponentTags []
      FilteredBones []
      FilteredSockets [FX_Socket_Eye_R, FX_Socket_Eye_L]
      MeshUserParameter {parameter {name=None, typeDefHandle {registeredTypeIndex=1}, varData []}}
      SamplingRegions []
     }}
    Socket Sampling Mode=::NewEnumerator5
   )
   &SolveForcesAndVelocity(Clamp Velocity=false, Limit Acceleration=false)
   &DynamicMaterialParameters(
    Index 0 Param 1=2
    Index 0 Param 3=1
    Index 0 Param 4=&RandomRangeFloat(Minimum=0, Maximum=1)
    Index 3 Param 1=100
    Index 3 Param 2=1
    Index 3 Param 4=0.2
    Module.Param0WriteEnabled=true
    Module.Param0WriteEnabled3=true
    Module.Param1WriteEnabled=true
    Module.Param1WriteEnabled3=true
    Module.Param2WriteEnabled=true
    Module.Param2WriteEnabled3=true
    Module.Param3WriteEnabled=true
    Module.Param3WriteEnabled3=true
    Write Parameter Index 3=true
   )
   &ScaleColor(
    Scale Alpha=&FloatFromCurve(FloatCurve=DataInterface() {propertyValues {Curve {keys [<at 0.7 0>,<at 1.5 1>]}}}, CurveIndex=$Emitter.Age)
    Scale RGB=&Multiply_VectorByFloat(
     Vector=&Multiply_VectorByFloat(
      Vector=&VectorFromFloat(Value=&FNTimeOfDayScalar(Night Scale=0.5, Morning Scale=0.5, Evening Scale=0.5))
      Float=&FN_FrontEndFloatModifier(Gameplay=1, Front End=0.2)
     )
     Float=1
    )
    ScaleA=true
    ScaleRGB=true
   )
   &FN_Light_Attributes(
    Color=<rgba 1 0.05 0.1292 1>
    Create Color Attribute=true
    Create Exponent Attribute=true
    Create Position Attribute=true
    Create Radius Attribute=true
    Enable Particle Light Enable=true
    Exponent=1.5
    Light Enabled=true
    Position=&AddVectorToPosition(
     Position=$Particles.Position
     Vector=&SelectVectorFromArray(Direct Array Index=$Particles.SpawnGroup, Vector Selection Array=DataInterface() {propertyValues {FloatData [<-2 0 0>, <2 0 0>]}})
    )
    Radius=20
    Scale Color=5
   )
  ]
  rendererClasses [&NiagaraSpriteRendererProperties]
  renderers [{rendererClass=&NiagaraSpriteRendererProperties}]
 }
 Eyes_SplineTrail_01 {
  emitterUpdateScript [
   &EmitterState(Life Cycle Mode=::NewEnumerator1, Loop Behavior=::NewEnumerator1, Loop Duration Mode=::NewEnumerator1, UseLoopDelay=true)
   &SpawnBurst_Instantaneous(Spawn Count=2, Use Spawn Probability=false, UseLoopCountLimit=false)
  ]
  particleSpawnScript [
   &FN_SetSpawnGroup(Number of Spawn Groups=2)
   &InitializeParticle(
    Mesh Scale=&Multiply_Vector(
     A=&SelectVectorFromArray(Direct Array Index=$Particles.SpawnGroup, Vector Selection Array=DataInterface() {propertyValues {FloatData [<1 1 1>, <-1 1 1>]}})
     B=<2 4 3>
    )
    Mesh Scale Mode=::NewEnumerator4
    UsePositionOffset=false
    Write Material Random=true
   )
   &InitialMeshOrientation(
    Mesh Axis To Orient=<0 1 0>
    Mesh Orientation Mode=::NewEnumerator2
    Rotation=&SelectVectorFromArray(Direct Array Index=$Particles.SpawnGroup, Vector Selection Array=DataInterface() {propertyValues {FloatData [<0 -0.025 0>, <0 0.0025 0>]}})
    Use Rotation Vector=true
   )
   &CameraOffset(Camera Offset Amount=1.5)
  ]
  particleUpdateScript [
   &ParticleState(Kill Particles When Lifetime Has Elapsed=false, Let Infinitely Lived Particles Die When Emitter Deactivates=true)
   &SkeletalMeshLocation(
    Bone Sampling Mode=::NewEnumerator4
    MaskSpawnGroup=false
    Sampled Position Offset=&FN_FrontEndVector3Modifier(
     Gameplay=&SelectVectorFromArray(Direct Array Index=$Particles.SpawnGroup, Vector Selection Array=DataInterface() {propertyValues {FloatData [<15 10 22>, <15 -10 22>]}})
     Front End=&SelectVectorFromArray(Direct Array Index=$Particles.SpawnGroup, Vector Selection Array=DataInterface() {propertyValues {FloatData [<5 5 22>, <5 -3 22>]}})
    )
    Skeletal Mesh=DataInterface() {propertyValues {
      DefaultMesh=null
      ExcludeBoneName=root
      OptionalComponentClass=null
      PreviewMesh=&Headset_Glider
      SkinningMode=SkinOnTheFly
      SoftSourceActor=null
      SourceMode=Default
      UvSetIndex=0
      WholeMeshLOD=-1
      bExcludeBone=false
      bReadDeformedGeometry=true
      bRequireCurrentFrameData=true
      ComponentTags []
      FilteredBones [JNT_head]
      FilteredSockets []
      MeshUserParameter {parameter {name=None, typeDefHandle {registeredTypeIndex=1}, varData []}}
      SamplingRegions []
     }}
   )
   &ScaleMeshSize(Scale Factor=&FN_FrontEndVector3Modifier(Gameplay=<1 1 1>, Front End=<1 0.5 1>))
   &InterpolateOverTime(
    Rate Of Change=8
    Target Vector Value:Vector3f=&ConvertPositionToVector(Input Position=$Particles.Position, Convert To Absolute Large World Vector=false)
    Variable Type=::NewEnumerator2
   )
   /SetVariables(Particles.Interpolate Dynamic=&MakeFloat4FromVectorAndFloat("Vector (RGB)"=$StackContext.InterpolateOverTime.InterpolateOverTime_V3_.Moving Average, "Float (Alpha)"=0.3)) {isSetParameters=true, module=SetVariables_#2}
   &ScaleColor(
    Scale Alpha=&Multiply_Float(
     A=&FloatFromCurve(FloatCurve=DataInterface() {propertyValues {Curve {keys [<at 0.7 0>,<at 1.5 1>]}}}, CurveIndex=$Emitter.Age)
     B=&FN_FrontEndFloatModifier(Gameplay=1, Front End=0.2)
    )
    Scale RGB=&Multiply_VectorByFloat(
     Vector=&Multiply_VectorByFloat(
      Vector=&VectorFromFloat(Value=&FNTimeOfDayScalar(Night Scale=0.5, Morning Scale=0.25, Evening Scale=0.25))
      Float=&FN_FrontEndFloatModifier(Gameplay=1, Front End=1)
     )
     Float=2
    )
    ScaleA=true
    ScaleRGB=true
   )
   &DynamicMaterialParameters(
    Index 0 Param 1=&RandomRangeFloat(Minimum=0, Maximum=1)
    Index 0 Param 3=1
    Index 0 Param 4=100
    Index 3 Param 1=$User.Grow
    Index 3 Param 2=3
    Module.Param0WriteEnabled=true
    Module.Param0WriteEnabled3=true
    Module.Param1WriteEnabled=true
    Module.Param1WriteEnabled3=true
    Module.Param2WriteEnabled=true
    Module.Param2WriteEnabled3=true
    Module.Param3WriteEnabled=true
    Module.Param3WriteEnabled3=true
    Write Parameter Index 3=true
   )
  ]
  rendererClasses [&NiagaraMeshRendererProperties]
  renderers [
   {rendererClass=&NiagaraMeshRendererProperties, rendererIndex=0}
   {rendererClass=&NiagaraMeshRendererProperties, rendererIndex=1}
  ]
 }
 Eyes_SplineTrail_02 {
  emitterUpdateScript [
   &EmitterState(Life Cycle Mode=::NewEnumerator1, Loop Behavior=::NewEnumerator1, Loop Duration Mode=::NewEnumerator1, UseLoopDelay=true)
   &SpawnBurst_Instantaneous(Spawn Count=2, Use Spawn Probability=false, UseLoopCountLimit=false)
  ]
  particleSpawnScript [
   &FN_SetSpawnGroup(Number of Spawn Groups=2)
   &InitializeParticle(
    Color=<rgba 1 1 1 0.5>
    Mesh Scale=&Multiply_Vector(
     A=&SelectVectorFromArray(Direct Array Index=$Particles.SpawnGroup, Vector Selection Array=DataInterface() {propertyValues {FloatData [<1 1 1>, <-1 1 1>]}})
     B=<0.8 2.8 3>
    )
    Mesh Scale Mode=::NewEnumerator4
    UsePositionOffset=false
    Write Material Random=true
   )
   &InitialMeshOrientation(
    Mesh Axis To Orient=<0 1 0>
    Mesh Orientation Mode=::NewEnumerator2
    Rotation=&SelectVectorFromArray(Direct Array Index=$Particles.SpawnGroup, Vector Selection Array=DataInterface() {propertyValues {FloatData [<0 -0.04 0>, <0 0.04 0>]}})
    Use Rotation Vector=true
   )
   &CameraOffset(Camera Offset Amount=1.5)
  ]
  particleUpdateScript [
   &ParticleState(Kill Particles When Lifetime Has Elapsed=false, Let Infinitely Lived Particles Die When Emitter Deactivates=true)
   &SkeletalMeshLocation(
    Bone Sampling Mode=::NewEnumerator4
    MaskSpawnGroup=false
    Sampled Position Offset=&FN_FrontEndVector3Modifier(
     Gameplay=&SelectVectorFromArray(Direct Array Index=$Particles.SpawnGroup, Vector Selection Array=DataInterface() {propertyValues {FloatData [<6 20 22>, <6 -20 22>]}})
     Front End=&SelectVectorFromArray(Direct Array Index=$Particles.SpawnGroup, Vector Selection Array=DataInterface() {propertyValues {FloatData [<2 10 18>, <2 -10 18>]}})
    )
    Skeletal Mesh=DataInterface() {propertyValues {
      DefaultMesh=null
      ExcludeBoneName=root
      OptionalComponentClass=null
      PreviewMesh=&Headset_Glider
      SkinningMode=SkinOnTheFly
      SoftSourceActor=null
      SourceMode=Default
      UvSetIndex=0
      WholeMeshLOD=-1
      bExcludeBone=false
      bReadDeformedGeometry=true
      bRequireCurrentFrameData=true
      ComponentTags []
      FilteredBones [JNT_head]
      FilteredSockets []
      MeshUserParameter {parameter {name=None, typeDefHandle {registeredTypeIndex=1}, varData []}}
      SamplingRegions []
     }}
   )
   &ScaleMeshSize(Scale Factor=&FN_FrontEndVector3Modifier(Gameplay=<1 1 1>, Front End=<1 0.5 1>))
   &InterpolateOverTime(
    Rate Of Change=8
    Target Vector Value:Vector3f=&ConvertPositionToVector(Input Position=$Particles.Position, Convert To Absolute Large World Vector=false)
    Variable Type=::NewEnumerator2
   )
   /SetVariables(Particles.Interpolate Dynamic=&MakeFloat4FromVectorAndFloat("Vector (RGB)"=$StackContext.InterpolateOverTime.InterpolateOverTime_V3_.Moving Average, "Float (Alpha)"=0.3)) {isSetParameters=true, module=SetVariables_#2}
   &ScaleColor(
    Scale Alpha=&FloatFromCurve(FloatCurve=DataInterface() {propertyValues {Curve {keys [<at 0.7 0>,<at 1.5 1>]}}}, CurveIndex=$Emitter.Age)
    Scale RGB=&Multiply_VectorByFloat(
     Vector=&Multiply_VectorByFloat(
      Vector=&VectorFromFloat(Value=&FNTimeOfDayScalar(Night Scale=0.5, Morning Scale=0.25, Evening Scale=0.25))
      Float=&FN_FrontEndFloatModifier(Gameplay=1, Front End=1)
     )
     Float=2
    )
    ScaleA=true
    ScaleRGB=true
   )
   &DynamicMaterialParameters(
    Index 0 Param 1=&RandomRangeFloat(Minimum=0, Maximum=1)
    Index 0 Param 3=1
    Index 0 Param 4=100
    Index 3 Param 1=$User.Grow
    Module.Param0WriteEnabled=true
    Module.Param0WriteEnabled3=true
    Module.Param1WriteEnabled=true
    Module.Param1WriteEnabled3=true
    Module.Param2WriteEnabled=true
    Module.Param2WriteEnabled3=true
    Module.Param3WriteEnabled=true
    Module.Param3WriteEnabled3=true
    Write Parameter Index 3=true
   )
  ]
  rendererClasses [&NiagaraMeshRendererProperties]
  renderers [{rendererClass=&NiagaraMeshRendererProperties, rendererIndex=0}]
 }
 Flame_Burst {
  emitterUpdateScript [
   &EmitterState(
    Inactive Response=::NewEnumerator1
    Life Cycle Mode=::NewEnumerator1
    Loop Behavior=::NewEnumerator1
    Loop Delay=0.9
    Loop Duration Mode=::NewEnumerator1
    UseLoopDelay=true
   )
   &SpawnBurst_Instantaneous(
    Spawn Count=22
    Spawn Probability=&FN_FrontEndFloatModifier(Gameplay=1, Front End=0)
    Use Spawn Probability=true
    UseLoopCountLimit=false
   )
  ]
  particleSpawnScript [
   &FN_SetSpawnGroup(Number of Spawn Groups=22)
   /SetVariables(Particles.New Normalize Spawn Group:NiagaraFloat=&Divide_Float(
     A=&MakeFloatFromInt(Int=&Multiply_Int(A=$Particles.SpawnGroup, B=2))
     B=43
    )) {isSetParameters=true, module=SetVariables_#3}
   &InitializeParticle(
    Color=<rgba 0.5733 0 0.8 3>
    Lifetime Max=1.6
    Lifetime Min=1.4
    Lifetime Mode=::NewEnumerator1
    Sprite Rotation Mode=::NewEnumerator1
    Sprite Size=&Multiply_Vector2D_ByFloat(
     Float=&Multiply_Float(
      A=&RandomRangeFloat(Minimum=45, Maximum=55)
      B=&FloatFromCurve(
       FloatCurve=DataInterface() {propertyValues {Curve {keys [
           {arriveTangent=6.4669, interpMode=RCIM_Cubic, leaveTangent=6.4669, tangentMode=RCTM_User, time=0, value=0.7}
           {arriveTangent=-0.4757, interpMode=RCIM_Cubic, leaveTangent=-3.2871, tangentMode=RCTM_Break, time=0.2, value=1.7}
           {arriveTangent=-0.7286, interpMode=RCIM_Cubic, leaveTangent=-0.7286, tangentMode=RCTM_User, time=0.5704, value=1.175}
           {arriveTangent=-0.5046, interpMode=RCIM_Cubic, leaveTangent=-0.5046, tangentMode=RCTM_User, time=1, value=0.875}
          ]}}}
       CurveIndex=$Particles.New Normalize Spawn Group
       Scale Curve=0.75
      )
     )
     Vector2D=<0.5 1>
    )
    Sprite Size Mode=::NewEnumerator4
    UsePositionOffset=false
    Write Material Random=true
   )
   &CameraOffset(Camera Offset Amount=80)
  ]
  particleUpdateScript [
   &ParticleState(
    Kill Particles When Lifetime Has Elapsed=false
    Let Infinitely Lived Particles Die When Emitter Deactivates=true
    Loop Particles Lifetime=true
   )
   &ScaleSpriteSize(
    Scale Sprite Size Mode=::NewEnumerator0
    Uniform Scale Factor=&Multiply_Float(
     A=&FloatFromCurve(
      FloatCurve=DataInterface() {propertyValues {Curve {keys [
          {interpMode=RCIM_Cubic, tangentMode=RCTM_User, time=0, value=0}
          {interpMode=RCIM_Cubic, tangentMode=RCTM_User, time=0.3, value=4}
          {arriveTangent=-3.1929, interpMode=RCIM_Cubic, leaveTangent=-3.1929, tangentMode=RCTM_User, time=0.8, value=0}
         ]}}}
      CurveIndex=&Add_Float(
       A=&Add_Float(A=$Emitter.Age, B=-0.9)
       B=&Multiply_Float(A=$Particles.New Normalize Spawn Group, B=-1.5)
      )
     )
     B=&FN_FrontEndFloatModifier(Gameplay=0.85, Front End=0.5)
    )
   )
   &SkeletalMeshLocation(
    Bone / Socket Index=&Multiply_Int(A=$Particles.SpawnGroup, B=2)
    Bone Sampling Mode=::NewEnumerator4
    MaskSpawnGroup=false
    Sampled Position Offset=&MakeVector(
     X=0
     Y=0
     Z=&FloatFromCurve(
      FloatCurve=DataInterface() {propertyValues {Curve {keys [<at 0 0>,<at 0.05 0.4524>,<at 0.2838 0.6524>,<at 0.45 0.2524>,<at 1 0>]}}}
      CurveIndex=$Particles.New Normalize Spawn Group
      Scale Curve=0
     )
    )
    Skeletal Mesh=$User.Skeletal Mesh
   )
   &SubUVAnimation(
    End Frame Range Override=32
    Random Start Frame=true
    "Sprite Renderer "=DataInterface() {propertyValues {SpriteRenderer=&self:Flame_Burst_11.NiagaraSpriteRendererProperties_0}}
    UseEndFrame=true
    UseStartFrame=true
   )
   &ScaleColor(
    Scale Alpha=&Multiply_Float(
     A=&FloatFromCurve(
      FloatCurve=DataInterface() {propertyValues {Curve {keys [
          {interpMode=RCIM_Cubic, tangentMode=RCTM_User, time=0, value=0}
          {interpMode=RCIM_Cubic, tangentMode=RCTM_User, time=0.2, value=1}
          <at 0.75 0>
         ]}}}
      CurveIndex=&Add_Float(
       A=&Add_Float(A=$Emitter.Age, B=-0.9)
       B=&Multiply_Float(A=$Particles.New Normalize Spawn Group, B=-1.5)
      )
     )
     B=1
    )
    Scale RGB=&VectorFromFloat(Value=&Multiply_Float(
      A=&Multiply_Float(
       A=&FloatFromCurve(FloatCurve=DataInterface() {propertyValues {Curve {keys [{arriveTangent=-34.6836, time=0, value=6}, <at 0.3 1>, <at 0.3 1>]}}})
       B=&FN_FrontEndFloatModifier(Gameplay=0.5, Front End=0.25)
      )
      B=&FNTimeOfDayScalar(Night Scale=2, Morning Scale=1, Evening Scale=1)
     ))
    ScaleA=true
    ScaleRGB=true
   )
  ]
  rendererClasses [&NiagaraSpriteRendererProperties]
  renderers [{rendererClass=&NiagaraSpriteRendererProperties}]
 }
 Flame_Burst_FE {
  emitterUpdateScript [
   &EmitterState(
    Inactive Response=::NewEnumerator1
    Life Cycle Mode=::NewEnumerator1
    Loop Behavior=::NewEnumerator1
    Loop Delay=0.8
    Loop Duration Mode=::NewEnumerator1
    UseLoopDelay=true
   )
   &SpawnBurst_Instantaneous(
    Spawn Count=44
    Spawn Probability=&FN_FrontEndFloatModifier(Gameplay=0, Front End=1)
    Use Spawn Probability=true
    UseLoopCountLimit=false
   )
  ]
  particleSpawnScript [
   &FN_SetSpawnGroup(Number of Spawn Groups=44)
   /SetVariables(Particles.New Normalize Spawn Group:NiagaraFloat=&Divide_Float(
     A=&MakeFloatFromInt(Int=&Multiply_Int(A=$Particles.SpawnGroup, B=2))
     B=43
    )) {isSetParameters=true, module=SetVariables_#3}
   &InitializeParticle(
    Color=&RandomRangeLinearColor(Minimum=<rgba 0.43 0 0.6 3>, Maximum=<rgba 0.86 0 1.2 1>)
    Lifetime Max=1.6
    Lifetime Min=1.4
    Lifetime Mode=::NewEnumerator1
    Sprite Rotation Mode=::NewEnumerator1
    Sprite Size=&Multiply_Vector2D_ByFloat(
     Float=&Multiply_Float(
      A=&RandomRangeFloat(Minimum=60, Maximum=80)
      B=&FloatFromCurve(
       FloatCurve=DataInterface() {propertyValues {Curve {keys [
           {arriveTangent=6.4669, interpMode=RCIM_Cubic, leaveTangent=6.4669, tangentMode=RCTM_User, time=0, value=0.7}
           {arriveTangent=-0.4757, interpMode=RCIM_Cubic, leaveTangent=-3.2871, tangentMode=RCTM_Break, time=0.2, value=1.7}
           {arriveTangent=-0.7286, interpMode=RCIM_Cubic, leaveTangent=-0.7286, tangentMode=RCTM_User, time=0.5704, value=1.175}
           {arriveTangent=-0.5046, interpMode=RCIM_Cubic, leaveTangent=-0.5046, tangentMode=RCTM_User, time=1, value=0.875}
          ]}}}
       CurveIndex=$Particles.New Normalize Spawn Group
       Scale Curve=0.75
      )
     )
     Vector2D=<0.5 1>
    )
    Sprite Size Mode=::NewEnumerator4
    Sprite UV Mode=::NewEnumerator1
    UsePositionOffset=false
    Write Material Random=true
   )
   &CameraOffset(Camera Offset Amount=80)
  ]
  particleUpdateScript [
   &ParticleState(
    Kill Particles When Lifetime Has Elapsed=false
    Let Infinitely Lived Particles Die When Emitter Deactivates=true
    Loop Particles Lifetime=true
   )
   &ScaleSpriteSize(
    Scale Sprite Size Mode=::NewEnumerator0
    Uniform Scale Factor=&Multiply_Float(
     A=&FloatFromCurve(
      FloatCurve=DataInterface() {propertyValues {Curve {keys [
          {interpMode=RCIM_Cubic, tangentMode=RCTM_User, time=0, value=0}
          {interpMode=RCIM_Cubic, tangentMode=RCTM_User, time=0.3, value=6}
          {arriveTangent=-3.1929, interpMode=RCIM_Cubic, leaveTangent=-3.1929, tangentMode=RCTM_User, time=0.8, value=0}
         ]}}}
      CurveIndex=&Add_Float(
       A=&Add_Float(A=$Emitter.Age, B=-0.8)
       B=&Multiply_Float(A=$Particles.New Normalize Spawn Group, B=-0.9)
      )
     )
     B=&FN_FrontEndFloatModifier(Gameplay=0.85, Front End=0.5)
    )
   )
   &ScaleColor(
    Scale Alpha=&Multiply_Float(
     A=&FloatFromCurve(
      FloatCurve=DataInterface() {propertyValues {Curve {keys [
          {interpMode=RCIM_Cubic, tangentMode=RCTM_User, time=0, value=0}
          {interpMode=RCIM_Cubic, tangentMode=RCTM_User, time=0.2, value=1}
          <at 0.75 0>
         ]}}}
      CurveIndex=&Add_Float(
       A=&Add_Float(A=$Emitter.Age, B=-0.8)
       B=&Multiply_Float(A=$Particles.New Normalize Spawn Group, B=-0.85)
      )
     )
     B=1
    )
    Scale RGB=&VectorFromFloat(Value=&Multiply_Float(
      A=&Multiply_Float(
       A=&FloatFromCurve(FloatCurve=DataInterface() {propertyValues {Curve {keys [{arriveTangent=-34.6836, time=0, value=1}, <at 0.3 1>, <at 0.3 1>]}}})
       B=&FN_FrontEndFloatModifier(Gameplay=0.5, Front End=0.25)
      )
      B=&FNTimeOfDayScalar(Night Scale=2, Morning Scale=1, Evening Scale=1)
     ))
    ScaleA=true
    ScaleRGB=true
   )
   &SkeletalMeshLocation(
    Bone / Socket Index=&Multiply_Int(A=$Particles.SpawnGroup, B=1)
    Bone Sampling Mode=::NewEnumerator4
    MaskSpawnGroup=false
    Sampled Position Offset=&MakeVector(
     X=0
     Y=0
     Z=&FloatFromCurve(
      FloatCurve=DataInterface() {propertyValues {Curve {keys [<at 0 0>,<at 0.05 0.4524>,<at 0.2838 0.6524>,<at 0.45 0.2524>,<at 1 0>]}}}
      CurveIndex=$Particles.New Normalize Spawn Group
      Scale Curve=0
     )
    )
    Skeletal Mesh=$User.Skeletal Mesh
   )
   &SubUVAnimation(
    End Frame Range Override=32
    Random Start Frame=true
    "Sprite Renderer "=DataInterface() {propertyValues {SpriteRenderer=&self:Flame_Burst_FE_2.NiagaraSpriteRendererProperties_0}}
    UseEndFrame=true
    UseStartFrame=true
   )
  ]
  rendererClasses [&NiagaraSpriteRendererProperties]
  renderers [{rendererClass=&NiagaraSpriteRendererProperties}]
 }
 Flame_Rainbow {
  emitterUpdateScript [
   &EmitterState(Inactive Response=::NewEnumerator1, Life Cycle Mode=::NewEnumerator1, Loop Delay=2, Loop Duration=1, UseLoopDelay=true)
   &SpawnRate(SpawnRate=30, Use Spawn Probability=false)
  ]
  particleSpawnScript [
   &FN_SetSpawnGroup(Number of Spawn Groups=21)
   /SetVariables(Particles.New Normalize Spawn Group:NiagaraFloat=&Divide_Float(
     A=&MakeFloatFromInt(Int=&Multiply_Int(A=$Particles.SpawnGroup, B=2))
     B=43
    )) {isSetParameters=true, module=SetVariables_#1}
   &InitializeParticle(
    Color=<rgba 1.5 0.15 1.5 0.75>
    Lifetime Max=2.5
    Lifetime Min=1.4
    Lifetime Mode=::NewEnumerator1
    Sprite Size=&Multiply_Vector2D_ByFloat(
     Float=&Multiply_Float(
      A=&RandomRangeFloat(Minimum=140, Maximum=120)
      B=&FloatFromCurve(
       FloatCurve=DataInterface() {propertyValues {Curve {keys [
           {arriveTangent=6.4669, interpMode=RCIM_Cubic, leaveTangent=6.4669, tangentMode=RCTM_User, time=0, value=0.7}
           {arriveTangent=-0.4757, interpMode=RCIM_Cubic, leaveTangent=-3.2871, tangentMode=RCTM_Break, time=0.2105, value=2.9468}
           {arriveTangent=-2.1574, interpMode=RCIM_Cubic, leaveTangent=-2.1574, tangentMode=RCTM_User, time=0.55, value=1.6581}
           {arriveTangent=-0.5046, interpMode=RCIM_Cubic, leaveTangent=-0.5046, tangentMode=RCTM_User, time=0.9, value=1.25}
           <at 1 0>
          ]}}}
       CurveIndex=$Particles.New Normalize Spawn Group
      )
     )
     Vector2D=<1 1>
    )
    Sprite Size Mode=::NewEnumerator4
    UsePositionOffset=false
    Write Material Random=true
   )
   &CameraOffset(Camera Offset Amount=&FloatFromCurve(
     FloatCurve=DataInterface() {propertyValues {Curve {keys [
         <at 0 0.7>
         {interpMode=RCIM_Cubic, leaveTangent=-2.3217, tangentMode=RCTM_Break, time=0.2, value=1}
         {interpMode=RCIM_Cubic, tangentMode=RCTM_User, time=1, value=0.3}
        ]}}}
     CurveIndex=$Particles.New Normalize Spawn Group
     Scale Curve=30
    ))
  ]
  particleUpdateScript [
   &ParticleState()
   &ScaleSpriteSize(
    Scale Sprite Size Mode=::NewEnumerator0
    Uniform Scale Factor=&Multiply_Float(
     A=&FloatFromCurve(FloatCurve=DataInterface() {propertyValues {Curve {keys [
          {arriveTangent=1.9809, interpMode=RCIM_Cubic, leaveTangent=1.9809, tangentMode=RCTM_User, time=0, value=0}
          {arriveTangent=0.6204, interpMode=RCIM_Cubic, leaveTangent=0.6204, tangentMode=RCTM_User, time=0.2, value=0.65}
          <at 1 1>
         ]}}})
     B=&FN_FrontEndFloatModifier(Gameplay=1, Front End=0.65)
    )
   )
   &SkeletalMeshLocation(
    Bone / Socket Index=&Multiply_Int(A=$Particles.SpawnGroup, B=2)
    Bone Sampling Mode=::NewEnumerator4
    MaskSpawnGroup=false
    Sampled Position Offset=&MakeVector(
     X=0
     Y=0
     Z=&Add_Float(
      A=&FloatFromCurve(
       FloatCurve=DataInterface() {propertyValues {Curve {keys [<at 0 0>,<at 0.05 0.4524>,<at 0.2838 0.6524>,<at 0.45 0.2524>,<at 1 0>]}}}
       CurveIndex=$Particles.New Normalize Spawn Group
       Scale Curve=5
      )
      B=&Lerp_Float(A=0, B=&RandomRangeFloat(Minimum=50, Maximum=70), Alpha=$Particles.NormalizedAge)
     )
    )
    Skeletal Mesh=$User.Skeletal Mesh
   )
   &ScaleColor(
    Scale Alpha=&Multiply_Float(
     A=&FloatFromCurve(FloatCurve=DataInterface() {propertyValues {Curve {keys [
          {interpMode=RCIM_Cubic, tangentMode=RCTM_User, time=0, value=0}
          {interpMode=RCIM_Cubic, tangentMode=RCTM_User, time=0.4, value=1}
          {interpMode=RCIM_Cubic, tangentMode=RCTM_User, time=1, value=0}
         ]}}})
     B=&Multiply_Float(
      A=&FNTimeOfDayScalar(Night Scale=0.75, Morning Scale=0.15, Day Scale=1.5, Evening Scale=0.05)
      B=&FN_FrontEndFloatModifier(Gameplay=0.5, Front End=0.65)
     )
    )
    Scale RGB=&Multiply_VectorByFloat(Vector=<1 1 1>, Float=&FN_FrontEndFloatModifier(Gameplay=1, Front End=1))
    ScaleA=true
    ScaleRGB=true
   )
   &DynamicMaterialParameters(
    Index 0 Param 1=&RandomRangeFloat(Minimum=0, Maximum=1)
    Module.Param0WriteEnabled=true
    Module.Param1WriteEnabled=true
    Module.Param2WriteEnabled=true
    Module.Param3WriteEnabled=true
   )
  ]
  rendererClasses [&NiagaraSpriteRendererProperties]
  renderers [{rendererClass=&NiagaraSpriteRendererProperties}]
 }
 Flame_Red001 {
  emitterUpdateScript [
   &EmitterState(
    Inactive Response=::NewEnumerator1
    Life Cycle Mode=::NewEnumerator1
    Loop Behavior=::NewEnumerator1
    Loop Delay=1.03
    Loop Duration Mode=::NewEnumerator1
    UseLoopDelay=true
   )
   &SpawnBurst_Instantaneous(Spawn Count=22, Use Spawn Probability=false, UseLoopCountLimit=false)
  ]
  particleSpawnScript [
   &FN_SetSpawnGroup(Number of Spawn Groups=21)
   /SetVariables(Particles.New Normalize Spawn Group:NiagaraFloat=&Divide_Float(
     A=&MakeFloatFromInt(Int=&Multiply_Int(A=$Particles.SpawnGroup, B=2))
     B=43
    )) {isSetParameters=true, module=SetVariables_#1}
   &InitializeParticle(
    Color=<rgba 3 0 2.75 1>
    Lifetime Min=1.4
    Lifetime Mode=::NewEnumerator1
    Sprite Size=&Multiply_Vector2D_ByFloat(
     Float=&Multiply_Float(
      A=&RandomRangeFloat(Minimum=40, Maximum=65)
      B=&FloatFromCurve(
       FloatCurve=DataInterface() {propertyValues {Curve {keys [
           {arriveTangent=6.4669, interpMode=RCIM_Cubic, leaveTangent=6.4669, tangentMode=RCTM_User, time=0, value=0.7}
           {arriveTangent=-0.4757, interpMode=RCIM_Cubic, leaveTangent=-3.2871, tangentMode=RCTM_Break, time=0.2, value=1.7}
           {arriveTangent=-0.7286, interpMode=RCIM_Cubic, leaveTangent=-0.7286, tangentMode=RCTM_User, time=0.55, value=0.888}
           {arriveTangent=-0.5046, interpMode=RCIM_Cubic, leaveTangent=-0.5046, tangentMode=RCTM_User, time=0.9, value=0.6}
           <at 1 0>
          ]}}}
       CurveIndex=$Particles.New Normalize Spawn Group
      )
     )
     Vector2D=<1.2 1.2>
    )
    Sprite Size Mode=::NewEnumerator4
    UsePositionOffset=false
    Write Material Random=true
   )
   &CameraOffset(Camera Offset Amount=&FloatFromCurve(
     FloatCurve=DataInterface() {propertyValues {Curve {keys [
         <at 0 0.7>
         {interpMode=RCIM_Cubic, leaveTangent=-2.3217, tangentMode=RCTM_Break, time=0.2, value=1}
         {interpMode=RCIM_Cubic, tangentMode=RCTM_User, time=1, value=0.3}
        ]}}}
     CurveIndex=$Particles.New Normalize Spawn Group
     Scale Curve=20
    ))
  ]
  particleUpdateScript [
   &ParticleState(
    Kill Particles When Lifetime Has Elapsed=false
    Let Infinitely Lived Particles Die When Emitter Deactivates=true
    Loop Particles Lifetime=true
   )
   &ScaleSpriteSize(
    Scale Sprite Size Mode=::NewEnumerator0
    Uniform Scale Factor=&Multiply_Float(
     A=&FloatFromCurve(
      FloatCurve=DataInterface() {propertyValues {Curve {keys [{tangentMode=RCTM_User, time=0, value=0}, <at 0.2 1.7>, {tangentMode=RCTM_User, time=0.5, value=1}]}}}
      CurveIndex=&Add_Float(
       A=&Add_Float(A=$Emitter.Age, B=-1.03)
       B=&Multiply_Float(A=$Particles.New Normalize Spawn Group, B=-1.7)
      )
     )
     B=&FN_FrontEndFloatModifier(Gameplay=1, Front End=0.65)
    )
   )
   &SkeletalMeshLocation(
    Bone / Socket Index=&Multiply_Int(A=$Particles.SpawnGroup, B=2)
    Bone Sampling Mode=::NewEnumerator4
    MaskSpawnGroup=false
    Sampled Position Offset=&MakeVector(
     X=0
     Y=0
     Z=&FloatFromCurve(
      FloatCurve=DataInterface() {propertyValues {Curve {keys [<at 0 0>,<at 0.05 0.4524>,<at 0.2838 0.6524>,<at 0.45 0.2524>,<at 1 0>]}}}
      CurveIndex=$Particles.New Normalize Spawn Group
      Scale Curve=40
     )
    )
    Skeletal Mesh=$User.Skeletal Mesh
   )
   &SubUVAnimation(
    "Sprite Renderer "=DataInterface() {propertyValues {SpriteRenderer=&self:Flame_Red001_9.NiagaraSpriteRendererProperties_0}}
    Start Frame Range Override=1
    UseEndFrame=true
    UseStartFrame=true
   )
   &ScaleColor(
    Scale Alpha=&FloatFromCurve(
     FloatCurve=DataInterface() {propertyValues {Curve {keys [
         <at 0 0>
         {arriveTangent=0.1248, interpMode=RCIM_Cubic, leaveTangent=0.1248, tangentMode=RCTM_User, time=0.3, value=1}
         {interpMode=RCIM_Cubic, tangentMode=RCTM_User, time=1, value=0}
        ]}}}
     Scale Curve=&FN_FrontEndFloatModifier(Gameplay=1, Front End=0)
    )
    ScaleA=true
    ScaleRGB=true
   )
  ]
  rendererClasses [&NiagaraSpriteRendererProperties]
  renderers [{rendererClass=&NiagaraSpriteRendererProperties}]
 }
 Flame_White {
  emitterUpdateScript [
   &EmitterState(
    Inactive Response=::NewEnumerator1
    Life Cycle Mode=::NewEnumerator1
    Loop Behavior=::NewEnumerator1
    Loop Delay=1.03
    Loop Duration Mode=::NewEnumerator1
    UseLoopDelay=true
   )
   &SpawnBurst_Instantaneous(Spawn Count=21, Use Spawn Probability=false, UseLoopCountLimit=false)
  ]
  particleSpawnScript [
   &FN_SetSpawnGroup(Number of Spawn Groups=21)
   /SetVariables(Particles.New Normalize Spawn Group:NiagaraFloat=&Divide_Float(
     A=&MakeFloatFromInt(Int=&Multiply_Int(A=$Particles.SpawnGroup, B=2))
     B=43
    )) {isSetParameters=true, module=SetVariables_#1}
   &InitializeParticle(
    Color=<rgba 10 2.5 9.125 2>
    Lifetime Max=1.4
    Lifetime Min=0.8
    Lifetime Mode=::NewEnumerator1
    Sprite Size=&Multiply_Vector2D_ByFloat(
     Float=&Multiply_Float(
      A=&RandomRangeFloat(Minimum=40, Maximum=55)
      B=&FloatFromCurve(
       FloatCurve=DataInterface() {propertyValues {Curve {keys [
           {arriveTangent=6.4669, interpMode=RCIM_Cubic, leaveTangent=6.4669, tangentMode=RCTM_User, time=0, value=0.7}
           {arriveTangent=-0.4757, interpMode=RCIM_Cubic, leaveTangent=-3.2871, tangentMode=RCTM_Break, time=0.2, value=1.7}
           {arriveTangent=-0.7286, interpMode=RCIM_Cubic, leaveTangent=-0.7286, tangentMode=RCTM_User, time=0.55, value=0.888}
           {arriveTangent=-0.5046, interpMode=RCIM_Cubic, leaveTangent=-0.5046, tangentMode=RCTM_User, time=0.9, value=0.6}
           <at 1 0>
          ]}}}
       CurveIndex=$Particles.New Normalize Spawn Group
      )
     )
     Vector2D=<0.85 0.7>
    )
    Sprite Size Mode=::NewEnumerator4
    UsePositionOffset=false
    Write Material Random=true
   )
   &CameraOffset(Camera Offset Amount=&FloatFromCurve(
     FloatCurve=DataInterface() {propertyValues {Curve {keys [
         <at 0 0.7>
         {interpMode=RCIM_Cubic, leaveTangent=-2.3217, tangentMode=RCTM_Break, time=0.2, value=1}
         {interpMode=RCIM_Cubic, tangentMode=RCTM_User, time=1, value=0.3}
        ]}}}
     CurveIndex=$Particles.New Normalize Spawn Group
     Scale Curve=20
    ))
  ]
  particleUpdateScript [
   &ParticleState(
    Kill Particles When Lifetime Has Elapsed=false
    Let Infinitely Lived Particles Die When Emitter Deactivates=true
    Loop Particles Lifetime=true
   )
   &ScaleSpriteSize(
    Scale Sprite Size Mode=::NewEnumerator0
    Uniform Scale Factor=&Multiply_Float(
     A=&FloatFromCurve(
      FloatCurve=DataInterface() {propertyValues {Curve {keys [{tangentMode=RCTM_User, time=0, value=0}, <at 0.2 1.7>, {tangentMode=RCTM_User, time=0.5, value=1}]}}}
      CurveIndex=&Add_Float(
       A=&Add_Float(A=$Emitter.Age, B=-1.03)
       B=&Multiply_Float(A=$Particles.New Normalize Spawn Group, B=-1.7)
      )
     )
     B=&FN_FrontEndFloatModifier(Gameplay=1, Front End=0.65)
    )
   )
   &SkeletalMeshLocation(
    Bone / Socket Index=&Add_Integer(A=&Multiply_Int(A=$Particles.SpawnGroup, B=2), B=1)
    Bone Sampling Mode=::NewEnumerator4
    MaskSpawnGroup=false
    Sampled Position Offset=&MakeVector(
     X=0
     Y=0
     Z=&FloatFromCurve(
      FloatCurve=DataInterface() {propertyValues {Curve {keys [<at 0 0>,<at 0.05 0.4524>,<at 0.2838 0.6524>,<at 0.45 0.2524>,<at 1 0>]}}}
      CurveIndex=$Particles.New Normalize Spawn Group
      Scale Curve=20
     )
    )
    Skeletal Mesh=$User.Skeletal Mesh
   )
   &SubUVAnimation(
    "Sprite Renderer "=DataInterface() {propertyValues {SpriteRenderer=&self:Flame_White_9.NiagaraSpriteRendererProperties_0}}
    UseEndFrame=true
    UseStartFrame=true
   )
   &ScaleColor(Scale Alpha=&FN_FrontEndFloatModifier(Gameplay=1, Front End=0.5), ScaleA=true, ScaleRGB=true)
  ]
  rendererClasses [&NiagaraSpriteRendererProperties]
  renderers [{rendererClass=&NiagaraSpriteRendererProperties}]
 }
 Frag001 {
  bEnabled=false
  emitterUpdateScript [
   &EmitterState(
    Life Cycle Mode=::NewEnumerator1
    Loop Behavior=::NewEnumerator1
    Loop Delay=&FN_FrontEndFloatModifier(Gameplay=0.35, Front End=1.25)
    Loop Duration=1.5
    UseLoopDelay=true
   )
   &SpawnRate(
    SpawnRate=&FloatFromCurve(
     FloatCurve=DataInterface() {propertyValues {Curve {keys [{arriveTangent=1.9285, interpMode=RCIM_Cubic, leaveTangent=1.9285, tangentMode=RCTM_User, time=0, value=0}, <at 1 1>]}}}
     CurveIndex=$Emitter.NormalizedLoopAge
     Scale Curve=50
    )
    Use Spawn Probability=false
   )
  ]
  particleSpawnScript [
   &FN_SetSpawnGroup(Number of Spawn Groups=44)
   /SetVariables(Particles.New Normalize Spawn Group:NiagaraFloat=&Divide_Float(
     A=&MakeFloatFromInt(Int=&Multiply_Int(A=$Particles.SpawnGroup, B=1))
     B=43
    )) {isSetParameters=true, module=SetVariables_#3}
   &InitializeParticle(
    Color=<rgba 0.3 0.3 0.3 1>
    Lifetime Max=0.2
    Lifetime Min=0.1
    Lifetime Mode=::NewEnumerator1
    Mesh Scale Mode=::NewEnumerator1
    Mesh Uniform Scale Max=0.4
    Mesh Uniform Scale Min=0.6
    UsePositionOffset=false
    Write Material Random=true
   )
   &CameraOffset(Camera Offset Amount=80)
   &AddVelocity(Velocity=&RandomRangeVector(Minimum=<0 -100 0>, Maximum=<0 -300 0>))
   &InitialMeshOrientation(Use Rotation Vector=false)
   /SetVariables(Particles.MeshIndex=&RandomRangeInt(Minimum=0, Maximum=1)) {isSetParameters=true}
  ]
  particleUpdateScript [
   &ParticleState()
   &SkeletalMeshLocation(
    Bone Sampling Mode=::NewEnumerator0
    MaskSpawnGroup=false
    OverrideRandomSeed=false
    Sampled Position Offset=&Lerp_Vector(
     A=&RandomRangeVector(Minimum=<70 70 0>, Maximum=<-70 -70 0>)
     B=<0 0 0>
     Alpha=$Particles.NormalizedAge
    )
    Sampled Position Offset Coordinate Space:ENiagaraSkelMeshTransforms=::NewEnumerator1
    Skeletal Mesh=DataInterface() {propertyValues {
      DefaultMesh=null
      ExcludeBoneName=root
      OptionalComponentClass=null
      PreviewMesh=&Headset_Glider
      SkinningMode=SkinOnTheFly
      SoftSourceActor=null
      SourceMode=Default
      UvSetIndex=0
      WholeMeshLOD=-1
      bExcludeBone=false
      bReadDeformedGeometry=true
      bRequireCurrentFrameData=true
      ComponentTags []
      FilteredBones [JNT_L_foot,JNT_R_foot,JNT_spine_34,JNT_spine_38,JNT_spine_44]
      FilteredSockets []
      MeshUserParameter {parameter {name=None, typeDefHandle {registeredTypeIndex=1}, varData []}}
      SamplingRegions []
     }}
   )
   &ScaleMeshSize(Scale Factor=&VectorFromFloat(Value=&FloatFromCurve(FloatCurve=DataInterface() {propertyValues {Curve {keys [
          {interpMode=RCIM_Cubic, tangentMode=RCTM_User, time=0, value=1}
          {arriveTangent=-2.767, interpMode=RCIM_Cubic, leaveTangent=-2.767, tangentMode=RCTM_User, time=1, value=0}
         ]}}})))
   &SubUVAnimation(
    Random Start Frame=true
    "Sprite Renderer "=DataInterface() {propertyValues {SpriteRenderer=&self:Frag001_1.NiagaraSpriteRendererProperties_0}}
    UseEndFrame=true
    UseStartFrame=true
   )
   &DynamicMaterialParameters(
    Index 0 Param 2=1
    Index 0 Param 3=1
    Index 0 Param 4=1
    Module.Param0WriteEnabled=true
    Module.Param0WriteEnabled1=true
    Module.Param1WriteEnabled=true
    Module.Param1WriteEnabled1=true
    Module.Param2WriteEnabled=true
    Module.Param2WriteEnabled1=true
    Module.Param3WriteEnabled=true
    Module.Param3WriteEnabled1=true
    Write Parameter Index 1=true
   ) {module=DynamicMaterialParameters001}
   &MeshRotationForce(Rotation=&RandomVector(Vector Scale=500))
   &SolveForcesAndVelocity(Clamp Velocity=false, Limit Acceleration=false)
  ]
  rendererClasses [&NiagaraMeshRendererProperties]
  renderers [{rendererClass=&NiagaraMeshRendererProperties, rendererIndex=0}]
 }
 Light {
  emitterUpdateScript [
   &EmitterState(Life Cycle Mode=::NewEnumerator1, Loop Behavior=::NewEnumerator1, Loop Delay=1, Loop Duration Mode=::NewEnumerator1, UseLoopDelay=true)
   &SpawnBurst_Instantaneous(Spawn Count=5, Use Spawn Probability=false, UseLoopCountLimit=false)
  ]
  particleSpawnScript [
   &InitializeParticle(Color=<rgba 14 1.4 12.95 1>, UsePositionOffset=false, Write Material Random=true)
   &FN_SetSpawnGroup(Number of Spawn Groups=5)
  ]
  particleUpdateScript [
   &ParticleState(Kill Particles When Lifetime Has Elapsed=false, Let Infinitely Lived Particles Die When Emitter Deactivates=true)
   &SkeletalMeshLocation(
    Bone / Socket Index=$Particles.SpawnGroup
    Bone Sampling Mode=::NewEnumerator4
    MaskSpawnGroup=false
    Sampled Position Offset=<0 0 30>
    Skeletal Mesh=DataInterface() {propertyValues {
      DefaultMesh=null
      ExcludeBoneName=root
      OptionalComponentClass=null
      PreviewMesh=&Headset_Glider
      SkinningMode=SkinOnTheFly
      SoftSourceActor=null
      SourceMode=Default
      UvSetIndex=0
      WholeMeshLOD=-1
      bExcludeBone=false
      bReadDeformedGeometry=true
      bRequireCurrentFrameData=true
      ComponentTags []
      FilteredBones [JNT_spine_04,JNT_spine_11,JNT_spine_22,JNT_spine_32]
      FilteredSockets []
      MeshUserParameter {parameter {name=None, typeDefHandle {registeredTypeIndex=1}, varData []}}
      SamplingRegions []
     }}
   )
   &ScaleColor(
    Scale RGB=&VectorFromFloat(Value=&FNTimeOfDayScalar(Night Scale=0.65, Morning Scale=1.15, Evening Scale=1.15))
    ScaleA=true
    ScaleRGB=true
   )
   &FN_Light_Attributes(Create Exponent Attribute=true, Create Radius Attribute=true, Exponent=2, Radius=150)
  ]
  rendererClasses [&NiagaraSpriteRendererProperties, &NiagaraLightRendererProperties]
  renderers [
   {rendererClass=&NiagaraSpriteRendererProperties}
   {rendererClass=&NiagaraLightRendererProperties, rendererIndex=1}
  ]
 }
 Tail_01 {
  emitterUpdateScript [
   &EmitterState(Life Cycle Mode=::NewEnumerator1, Loop Behavior=::NewEnumerator1, Loop Delay=2.55, Loop Duration Mode=::NewEnumerator1, UseLoopDelay=true)
   &SpawnBurst_Instantaneous(Use Spawn Probability=false, UseLoopCountLimit=false)
  ]
  particleSpawnScript [
   &InitializeParticle(Mesh Scale=<0.5 2 4>, Mesh Scale Mode=::NewEnumerator4, UsePositionOffset=false, Write Material Random=true)
   &InitialMeshOrientation(Mesh Axis To Orient=<0 1 0>, Mesh Orientation Mode=::NewEnumerator2, Rotation=<0.01 0 0>, Use Rotation Vector=true)
   &CameraOffset(Camera Offset Amount=10)
  ]
  particleUpdateScript [
   &ParticleState(
    Kill Particles When Lifetime Has Elapsed=false
    Let Infinitely Lived Particles Die When Emitter Deactivates=true
    Loop Particles Lifetime=true
   )
   &ScaleMeshSize(Scale Factor=&VectorFromFloat(Value=&FloatFromCurve(
      FloatCurve=DataInterface() {propertyValues {Curve {keys [<at 2.55 0>,<at 2.8 1.5>,<at 3.5 1>]}}}
      CurveIndex=$Emitter.Age
      Scale Curve=&FN_FrontEndFloatModifier(Gameplay=1, Front End=0.5)
     )))
   &SkeletalMeshLocation(
    Bone Sampling Mode=::NewEnumerator4
    MaskSpawnGroup=false
    Sampled Position Offset=<-15 0 0.5>
    Skeletal Mesh=DataInterface() {propertyValues {
      DefaultMesh=null
      ExcludeBoneName=root
      OptionalComponentClass=null
      PreviewMesh=&Headset_Glider
      SkinningMode=SkinOnTheFly
      SoftSourceActor=null
      SourceMode=Default
      UvSetIndex=0
      WholeMeshLOD=-1
      bExcludeBone=false
      bReadDeformedGeometry=true
      bRequireCurrentFrameData=true
      ComponentTags []
      FilteredBones [JNT_spine_45]
      FilteredSockets []
      MeshUserParameter {parameter {name=None, typeDefHandle {registeredTypeIndex=1}, varData []}}
      SamplingRegions []
     }}
   )
   &InterpolateOverTime(
    Rate Of Change=8
    Target Vector Value:Vector3f=&ConvertPositionToVector(Input Position=$Particles.Position, Convert To Absolute Large World Vector=false)
    Variable Type=::NewEnumerator2
   )
   /SetVariables(Particles.Interpolate Dynamic=&MakeFloat4FromVectorAndFloat("Vector (RGB)"=$StackContext.InterpolateOverTime.InterpolateOverTime_V3_.Moving Average, "Float (Alpha)"=0.5)) {isSetParameters=true, module=SetVariables_#2}
  ]
  rendererClasses [&NiagaraMeshRendererProperties]
  renderers [{rendererClass=&NiagaraMeshRendererProperties, rendererIndex=0}]
 }
 Tail_02 {
  emitterUpdateScript [
   &EmitterState(Life Cycle Mode=::NewEnumerator1, Loop Behavior=::NewEnumerator1, Loop Delay=2.55, Loop Duration Mode=::NewEnumerator1, UseLoopDelay=true)
   &SpawnBurst_Instantaneous(Use Spawn Probability=false, UseLoopCountLimit=false)
  ]
  particleSpawnScript [
   &InitializeParticle(Mesh Scale=<1 1.7 2.5>, Mesh Scale Mode=::NewEnumerator4, UsePositionOffset=false, Write Material Random=true)
   &InitialMeshOrientation(Mesh Axis To Orient=<0 1 0>, Mesh Orientation Mode=::NewEnumerator2, Rotation=<-0.01 0 0>, Use Rotation Vector=true)
   &CameraOffset(Camera Offset Amount=10)
  ]
  particleUpdateScript [
   &ParticleState(Kill Particles When Lifetime Has Elapsed=false, Let Infinitely Lived Particles Die When Emitter Deactivates=true)
   &ScaleMeshSize(Scale Factor=&VectorFromFloat(Value=&FloatFromCurve(
      FloatCurve=DataInterface() {propertyValues {Curve {keys [<at 2.55 0>,<at 2.8 1.5>,<at 3.5 1>]}}}
      CurveIndex=$Emitter.Age
      Scale Curve=&FN_FrontEndFloatModifier(Gameplay=1, Front End=0.5)
     )))
   &SkeletalMeshLocation(
    Bone Sampling Mode=::NewEnumerator4
    MaskSpawnGroup=false
    Sampled Position Offset=<-20 0 0.5>
    Skeletal Mesh=DataInterface() {propertyValues {
      DefaultMesh=null
      ExcludeBoneName=root
      OptionalComponentClass=null
      PreviewMesh=&Headset_Glider
      SkinningMode=SkinOnTheFly
      SoftSourceActor=null
      SourceMode=Default
      UvSetIndex=0
      WholeMeshLOD=-1
      bExcludeBone=false
      bReadDeformedGeometry=true
      bRequireCurrentFrameData=true
      ComponentTags []
      FilteredBones [JNT_spine_45]
      FilteredSockets []
      MeshUserParameter {parameter {name=None, typeDefHandle {registeredTypeIndex=1}, varData []}}
      SamplingRegions []
     }}
   )
   &InterpolateOverTime(
    Rate Of Change=8
    Target Vector Value:Vector3f=&ConvertPositionToVector(Input Position=$Particles.Position, Convert To Absolute Large World Vector=false)
    Variable Type=::NewEnumerator2
   )
   /SetVariables(Particles.Interpolate Dynamic=&MakeFloat4FromVectorAndFloat("Vector (RGB)"=$StackContext.InterpolateOverTime.InterpolateOverTime_V3_.Moving Average, "Float (Alpha)"=0.5)) {isSetParameters=true, module=SetVariables_#2}
   &ScaleColor(
    Scale RGB=&Multiply_VectorByFloat(
     Vector=&Multiply_VectorByFloat(
      Vector=&VectorFromFloat(Value=&FNTimeOfDayScalar(Night Scale=0.5, Morning Scale=0.25, Evening Scale=0.25))
      Float=&FN_FrontEndFloatModifier(Gameplay=1, Front End=4)
     )
     Float=2
    )
    ScaleA=true
    ScaleRGB=true
   )
  ]
  rendererClasses [&NiagaraMeshRendererProperties]
  renderers [{rendererClass=&NiagaraMeshRendererProperties, rendererIndex=0}]
 }
}
summary {
 systemName=NS_Glider_Headset_Purple
 emitters [
  {emitterName=Eyes_SplineTrail_01, rendererClasses [&NiagaraMeshRendererProperties]}
  {emitterName=Eyes_SplineTrail_02, rendererClasses [&NiagaraMeshRendererProperties]}
  {emitterName=Eyes_Glow_01, rendererClasses [&NiagaraSpriteRendererProperties]}
  {emitterName=Tail_01, rendererClasses [&NiagaraMeshRendererProperties]}
  {emitterName=Tail_02, rendererClasses [&NiagaraMeshRendererProperties]}
  {emitterName=Flame_White, rendererClasses [&NiagaraSpriteRendererProperties]}
  {emitterName=Flame_Rainbow, rendererClasses [&NiagaraSpriteRendererProperties]}
  {emitterName=Ash, rendererClasses [&NiagaraSpriteRendererProperties]}
  {emitterName=Flame_Red001, rendererClasses [&NiagaraSpriteRendererProperties]}
  {bEnabled=false, emitterName=Frag001, rendererClasses [&NiagaraMeshRendererProperties]}
  {emitterName=Flame_Burst, rendererClasses [&NiagaraSpriteRendererProperties]}
  {emitterName=Flame_Burst_FE, rendererClasses [&NiagaraSpriteRendererProperties]}
  {emitterName=Light, rendererClasses [&NiagaraSpriteRendererProperties, &NiagaraLightRendererProperties]}
 ]
 userVariables [
  User.bIsFrontEnd {type=NiagaraBool}
  User.bIsFrontEndPreview {type=NiagaraBool}
  User.Grow {type=NiagaraFloat, defaultValue=1, description=""}
  User.Skeletal Mesh {
   type=NiagaraDataInterfaceSkeletalMesh
   defaultValue {
    struct=NiagaraExt_VariableValue_DataInterface
    value {dataInterface=&self:NiagaraDataInterfaceSkeletalMesh_0}
   }
  }
 ]
}
